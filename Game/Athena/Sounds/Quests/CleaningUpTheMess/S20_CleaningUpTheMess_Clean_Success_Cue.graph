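# Sound cue for the success feedback of cleaning in the Cleaning Up The Mess quest, routed through the Fort_UI sound class.
# A mixer layers an immediate random pick of two radio success stingers with a random pick of two sweeping sounds.
# The sweep layer passes through a modulator for slight variation and a fixed half-second delay.
# A concurrency override allows one instance at a time, stopping the oldest.
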SoundCue /Game/Athena/Sounds/Quests/CleaningUpTheMess/S20_CleaningUpTheMess_Clean_Success_Cue
root=Mixer_0
nodes [
 {id=Delay_0, children [Modulator_0], properties {delayMax=0.5, delayMin=0.5}}
 {id=Mixer_0, children [Random_1,Delay_0], properties {inputVolume [1,1]}}
 {id=Modulator_0, children [Random_0], properties {pitchMin=1, volumeMax=1.1, volumeMin=1}}
 {id=Random_0, children [WavePlayer_0,WavePlayer_1], properties {weights [1,1]}}
 {id=Random_1, children [WavePlayer_4,WavePlayer_5], properties {weights [1,1]}}
 {id=WavePlayer_0, properties {soundWaveAssetPtr=&/Spot_Sweeping_Up_01}}
 {id=WavePlayer_1, properties {soundWaveAssetPtr=&/Spot_Sweeping_Up_02}}
 {id=WavePlayer_4, properties {soundWaveAssetPtr=/Game/Sounds/Creative/Gadgets/Radio/Stingers/Stinger_Success_01}}
 {id=WavePlayer_5, properties {soundWaveAssetPtr=/Game/Sounds/Creative/Gadgets/Radio/Stingers/Stinger_Success_02}}
]
properties {
 bOverrideConcurrency=true
 maxDistance=2097152
 soundClassObject=&Fort_UI
 volumeMultiplier=1
 concurrencyOverrides {maxCount=1, resolutionRule=StopOldest}
}
