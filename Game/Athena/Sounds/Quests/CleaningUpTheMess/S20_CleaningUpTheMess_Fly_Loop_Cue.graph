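# Looping ambient Sound Cue of flies for the Cleaning Up The Mess quest, using a flies-around-trash attenuation with overrides and the quest sound class.
# A looping swarm bed runs through an indefinitely looping Enveloper whose cubic volume curve gently dips and recovers over twenty seconds.
# A second, much quieter looping branch plays one of five random buzz one-shots with pitch modulation after random short delays.
# A Mixer combines both branches, leaving its middle input empty.
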SoundCue /Game/Athena/Sounds/Quests/CleaningUpTheMess/S20_CleaningUpTheMess_Fly_Loop_Cue
root=Mixer_0
nodes [
 {id=Delay_0, children [Modulator_0], properties {delayMax=3, delayMin=1}}
 {
  id=Enveloper_2
  children [WavePlayer_1]
  properties {
   bLoop=true
   bLoopIndefinitely=true
   loopEnd=20
   pitchCurve {editorCurveData {keys [{time=0}]}}
   volumeCurve {editorCurveData {keys [
      {arriveTangent=-0.01, interpMode=RCIM_Cubic, time=0, value=1}
      <at 10 0.8 RCIM_Cubic>
      {interpMode=RCIM_Cubic, leaveTangent=0.01, time=20, value=1}
     ]}}
  }
 }
 {id=Looping_0, children [Delay_0]}
 {id=Mixer_0, children [Enveloper_2,null,Looping_0], properties {inputVolume [1.5,1.1,0.2]}}
 {id=Modulator_0, children [Random_0], properties {pitchMax=1.2}}
 {id=Random_0, children [WavePlayer_2,WavePlayer_3,WavePlayer_4,WavePlayer_5,WavePlayer_6], properties {weights [1,1,1,1,1]}}
 {id=WavePlayer_1, properties {bLooping=true, soundWaveAssetPtr=&/Spot_Fly_Swamming_By_Stink_Loop}}
 {id=WavePlayer_2, properties {soundWaveAssetPtr=&/Spot_Fly_Buzz_01}}
 {id=WavePlayer_3, properties {soundWaveAssetPtr=&/Spot_Fly_Buzz_02}}
 {id=WavePlayer_4, properties {soundWaveAssetPtr=&/Spot_Fly_Buzz_03}}
 {id=WavePlayer_5, properties {soundWaveAssetPtr=&/Spot_Fly_Buzz_04}}
 {id=WavePlayer_6, properties {soundWaveAssetPtr=&/Spot_Fly_Buzz_05}}
]
properties {
 attenuationSettings=&/Spot_Flies_Around_Trash_Att
 bPrimeOnLoad=true
 maxDistance=2010
 soundClassObject=&Fort_Quests
 volumeMultiplier=1
 attenuationOverrides {falloffDistance=700, attenuationShapeExtents {x=100}}
}
