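# Sound cue for the trash bag explosion in the S20 Cleaning Up The Mess quest.
# A Random node picks one of two equally weighted explosion waves and feeds it through a Modulator at default settings.
# Playback uses the mission in-game sound class and borrows the crystal shard mission attenuation asset.
# A concurrency override blocks new instances once the limit is reached, rather than stopping playing ones.
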SoundCue /Game/Athena/Sounds/Quests/CleaningUpTheMess/S20_CleaningUpTheMess_TrashBag_Explosion_Cue
root=Modulator_0
nodes [
 {id=Modulator_0, children [Random_0]}
 {id=Random_0, children [WavePlayer_0,WavePlayer_1], properties {weights [1,1]}}
 {id=WavePlayer_0, properties {soundWaveAssetPtr=&/Spot_Trash_Bag_Explosion_01}}
 {id=WavePlayer_1, properties {soundWaveAssetPtr=&/Spot_Trash_Bag_Explosion_02}}
]
properties {
 attenuationSettings=/Game/Sounds/Attenuation/Fort_Mission_CrystalShard_Att
 bOverrideConcurrency=true
 maxDistance=2800
 soundClassObject=/Game/Sounds/FortSoundClassesAndModes/Fort_Mission_IngameSounds
 volumeMultiplier=1
 attenuationOverrides {falloffDistance=2600}
 concurrencyOverrides {resolutionRule=PreventNew}
}
